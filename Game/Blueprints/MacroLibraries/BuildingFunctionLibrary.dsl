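; Blueprint function library of shared building helpers for debug logging, Creative island resource requests and actor-type sweep tests.
; Log printing is gated by a log level, compared against the instigator's own level when it implements BlueprintDebuggingInterface.
; A resource manager request result sets a component's active state and visibility.
; The sweep test checks current overlaps first, then sphere-traces by object type and matches hit actors against class filters, including subclasses.
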(fn LogString
  (params (Instance%20Log%20Level :type BlueprintLogLevel%20Enum :default "NewEnumerator0") (LogString :type String) (PrintToScreen :type Boolean) (Instigator :type Object%20Reference))
  (bind _should_log (CallFunction|ShouldLogforInstigator Instance%20Log%20Level Instigator Instigator))
  (if _should_log
    (Development|PrintString Instigator LogString PrintToScreen)
    (else
      _))
  (return))

(fn RequestComponentResource
  (params (ResourceManager :type Creative%20Island%20Resource%20Manager%20Component%20Object%20Reference) (ResourceTag :type Name) (Component :type Actor%20Component%20Object%20Reference) (WorldLocation :type Vector))
  (bind _returnvalue (Class|CreativeIslandResourceManagerComponent|RequestResource ResourceManager ResourceTag WorldLocation))
  (Components|Activation|SetActive Component _returnvalue true)
  (bind _asscene_component (Utilities|Casting|CastToSceneComponent Component))
  (Rendering|SetVisibility _asscene_component _returnvalue true))

(fn ShouldLogForInstigator
  (params (Instance%20Log%20Level :type BlueprintLogLevel%20Enum :default "NewEnumerator0") (Instigator :type Object%20Reference))
  (if (Utilities|Enum|NotEqual(Enum) Instance%20Log%20Level "NewEnumerator0")
    (if (Utilities|DoesObjectImplementInterface Instigator "/Game/Blueprints/Interfaces/BlueprintDebuggingInterface.BlueprintDebuggingInterface_C")
      (bind _blueprintloglevel (Class|BlueprintDebuggingInterface|GetBlueprintLogLevel(Message) Instigator))
      (return (Math|Byte|Byte<=Byte _blueprintloglevel Instance%20Log%20Level))
      (else
        (return true)))
    (else
      (return false))))

(fn SweepAgainstActorTypes
  (params (TestActor :type Actor%20Object%20Reference) (StartLocation :type Vector) (EndLocation :type Vector) (SweepRadius :type Float%20%28double-precision%29) (SweepObjectTypes :type Array%20of%20EObjectTypeQuery%20Enums :default "ObjectTypeQuery1") (ActorClassFilters :type Array%20of%20Actor%20Class%20References))
  ;[1 1 1] Check parameter validity.
  (Utilities|IsValid TestActor
    (:Is%20Valid
      (if (> (Utilities|Array|Length ActorClassFilters) 0)
        ;[1 1 1] Quick check to see if the actor is already overlapping one of the specified actor types.
        (for _array_element ActorClassFilters
          (if (Utilities|IsArrayNonEmpty (Collision|GetOverlappingActors (Variables|GetTestActor) _array_element :!class /Script/Engine.Actor))
            (return true)))
        (bind _outhits (Collision|MultiSphereTraceForObjects TestActor StartLocation EndLocation SweepRadius SweepObjectTypes))
        (for _array_element_2 _outhits
          (bind (_array_element_bblockinghit _array_element_binitialoverlap _array_element_time _array_element_distance _array_element_location _array_element_impactpoint _array_element_normal _array_element_impactnormal _array_element_physmat _array_element_hitactor) (Collision|BreakHitResult _array_element_2))
          (Utilities|IsValid _array_element_hitactor
            (:Is%20Valid
              ;[1 1 1] Determine if the overlapped actor is one of the specified actor types.
              (for _array_element_3 ActorClassFilters
                (if (Utilities|ClassIsChildOf (Utilities|GetClass _array_element_hitactor) _array_element_3)
                  (return true))))))
        (return false)
        (else
          _)))
    (:Is%20Not%20Valid
      _))
  (return false))
