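; Blueprint function library of shared helpers for Fortnite building props, covering scale limits, tinting, addon meshes and destructible glass setup.
; Tint functions pack a custom or building-texture-data color as HSV into one custom primitive data float.
; SpawnAddonMesh adds an optional static mesh component with material and collision options, skipping mobile.
; Other helpers clamp world scale per axis and hand provider components to SimpleDestructiblePropSubsystem for construction-script conversion.
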(fn ClampMeshScale
  (params (EnableScaleLimits :type Boolean) (MeshComponent :type Static%20Mesh%20Component%20Object%20Reference) (MinScale :type Float%20%28double-precision%29 :default 1.0) (MaxScale :type Float%20%28double-precision%29 :default 1.0))
  (if EnableScaleLimits
    (LocalVariables|SetL_MeshComponent MeshComponent)
    (LocalVariables|SetL_Min_Scale MinScale)
    (LocalVariables|SetL_MaxScale MaxScale)
    (bind _output_get (LocalVariables|SetL_MeshScale (Transformation|GetWorldScale (LocalVariables|GetL_MeshComponent))))
    (LocalVariables|SetL_X_Scale (.x _output_get))
    (LocalVariables|SetL_Y_Scale (.y _output_get))
    (LocalVariables|SetL_Z_Scale (.z _output_get))
    (Utilities|FlowControl|Sequence
      (:then_0
        (bind _output_get_2 (LocalVariables|SetL_X_Scale (Math|Float|Clamp(Float) (LocalVariables|GetL_X_Scale) (LocalVariables|GetL_Min_Scale) (LocalVariables|GetL_MaxScale)))))
      (:then_1
        (bind _output_get_3 (LocalVariables|SetL_Y_Scale (Math|Float|Clamp(Float) (LocalVariables|GetL_Y_Scale) (LocalVariables|GetL_Min_Scale) (LocalVariables|GetL_MaxScale)))))
      (:then_2
        (bind _output_get_4 (LocalVariables|SetL_Z_Scale (Math|Float|Clamp(Float) (LocalVariables|GetL_Z_Scale) (LocalVariables|GetL_Min_Scale) (LocalVariables|GetL_MaxScale)))))
      (:then_3
        (Transformation|SetWorldScale3D (LocalVariables|GetL_MeshComponent) (Math|Vector|MakeVector _output_get_2 _output_get_3 _output_get_4))
        (return)))))

(fn SetColorOnSecondaryMesh
  (params (Color :type Linear%20Color%20Structure) (Static%20Mesh :type Static%20Mesh%20Component%20Object%20Reference))
  (bind _returnvalue (Math|Color|RGBtoHSV(LinearColor) Color))
  (bind (_r _g _b) (Utilities|Struct|BreakLinearColor _returnvalue))
  (bind _returnvalue_2 (Math|Color|Fortnite|PackHSVtoFloat _r _g _b))
  (Rendering|Material|SetCustomPrimitiveDataFloat Static%20Mesh 30 _returnvalue_2)
  (Rendering|Material|SetDefaultCustomPrimitiveDataFloat Static%20Mesh 30 _returnvalue_2))

(fn SetTintColorOnLightingAsset
  (params (Static%20Mesh :type Static%20Mesh%20Component%20Object%20Reference) (TextureDataIndex :type Integer) (Target :type Building%20Static%20Mesh%20Object%20Reference) (SetCustomColor :type Boolean) (CustomColor :type Linear%20Color%20Structure :default "(R=1.000000,G=1.000000,B=1.000000,A=1.000000)"))
  (if SetCustomColor
    (Default|SetColoronSecondaryMesh CustomColor Static%20Mesh)
    (elif (Utilities|IsValid Target :!class /Script/Engine.KismetSystemLibrary)
      (bind _outtexturedataarray (Building|GetBuildingTextureData Target :!class /Script/FortniteGame.BuildingSMActor))
      (if (> (Utilities|Array|Length _outtexturedataarray) 0)
        (bind _outtexturedataarray_2 (Building|GetBuildingTextureData Target :!class /Script/FortniteGame.BuildingSMActor))
        (if (Utilities|IsValid (Utilities|Array|Get(acopy) _outtexturedataarray_2 TextureDataIndex) :!class /Script/Engine.KismetSystemLibrary)
          (Default|SetColoronSecondaryMesh (Math|Conversions|ToLinearColor(Color) (Variables|/Script/FortniteGame.BuildingTextureData|GetTintColor (Utilities|Array|Get(acopy) _outtexturedataarray_2 TextureDataIndex))) Static%20Mesh))))))

(fn SetTintColorOnSubMeshes
  (params (TintableMeshes :type Array%20of%20Static%20Mesh%20Component%20Object%20References) (TextureDataIndex :type Integer) (Target :type Building%20Static%20Mesh%20Object%20Reference) (SetCustomColor :type Boolean) (CustomColor :type Linear%20Color%20Structure :default "(R=1.000000,G=1.000000,B=1.000000,A=1.000000)"))
  (if SetCustomColor
    _
    (elif (Utilities|IsValid Target :!class /Script/Engine.KismetSystemLibrary)
      (bind _outtexturedataarray (Building|GetBuildingTextureData Target :!class /Script/FortniteGame.BuildingSMActor))
      (if (> (Utilities|Array|Length _outtexturedataarray) 0)
        (bind _outtexturedataarray_2 (Building|GetBuildingTextureData Target :!class /Script/FortniteGame.BuildingSMActor))
        (if (Utilities|IsValid (Utilities|Array|Get(acopy) _outtexturedataarray_2 TextureDataIndex) :!class /Script/Engine.KismetSystemLibrary)
          _))))
  (for _array_element TintableMeshes
    (Default|SetColoronSecondaryMesh (Math|Conversions|ToLinearColor(Color) (Variables|/Script/FortniteGame.BuildingTextureData|GetTintColor (Utilities|Array|Get(acopy) _outtexturedataarray_2 TextureDataIndex))) _array_element)))

(fn SpawnAddonMesh
  (params (SpawnAddonMesh :type Boolean :default true) (AddonMesh :type Static%20Mesh%20Soft%20Object%20Reference) (Target :type Actor%20Object%20Reference) (EnableCollision :type Boolean) (OverrideAddonMaterial :type Boolean) (OverrideMaterial :type Material%20Interface%20Soft%20Object%20Reference) (Location :type Vector))
  (bind _addoncomponent (LocalVariables|GetAddonComponent))
  (bind _output (Utilities|ResolveSoftReference _addoncomponent))
  ;[1 1 1] Bug with pre-assigning the mesh and then dragging the actor into the world. Have to re-assign the mesh to get it to populate
  ;;[1 1 1] Attempting to not spawn if mobile platform
  (if (Fortnite|IsMobilePlatform)
    _
    (elif (Utilities|IsValid Target :!class /Script/Engine.KismetSystemLibrary)
      (if SpawnAddonMesh
        (if (Utilities|IsValid (Utilities|ResolveSoftReference AddonMesh) :!class /Script/Engine.KismetSystemLibrary)
          (bind _returnvalue (Game|AddComponentbyClass Target 0 "/Script/Engine.StaticMeshComponent" false (Math|Transform|MakeTransform Location)))
          (if OverrideAddonMaterial
            (if (Utilities|IsValidSoftObjectReference OverrideMaterial)
              (Rendering|Material|SetMaterial _returnvalue 0 (Utilities|ResolveSoftReference OverrideMaterial))))
          (bind _output_get (LocalVariables|SetAddonComponent (Utilities|ResolveSoftReference _returnvalue)))
          (Transformation|SetMobility _output_get "Static")
          (if (not EnableCollision)
            (Collision|SetCollisionEnabled (Utilities|ResolveSoftReference _addoncomponent) "NoCollision"))
          (Variables|/Script/Engine.StaticMeshComponent|SetbEvaluateWorldPositionOffset false _output)
          (Components|StaticMesh|SetStaticMesh _output (Utilities|ResolveSoftReference AddonMesh))
          (else
            (Development|PrintString 0 "Missing assigned mesh")))
        (else
          (Components|DestroyComponent (Utilities|ResolveSoftReference (LocalVariables|GetAddonComponent)))
          (Development|PrintString 0 "Addon mesh removed"))))))

(fn ValidateGlassProviderComponents
  (params (Owner :type Actor%20Object%20Reference))
  (bind _returnvalue (Class|SimpleDestructiblePropSubsystem|GetSimpleDestructiblePropSubsystem))
  (Utilities|IsValid _returnvalue
    (:Is%20Valid
      (Class|SimpleDestructiblePropSubsystem|ConstructionScriptConverttoSimpleDestructiblePropData _returnvalue Owner (Actor|GetComponentsByClass Owner "/Script/FortniteGame.SimpleDestructiblePropProviderComponent"))
      (return))))
